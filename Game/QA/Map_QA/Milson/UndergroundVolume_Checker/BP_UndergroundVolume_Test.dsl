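; QA test actor that flags where the player stands outside every FortUndergroundVolume in a map.
; Each tick it moves a cube to the player pawn and loops over all those volumes for overlap.
; Without overlap it shows the cube and, unless HideDebugText is set, prints a warning.
; Alongside the warning, it copies a BugItGo command with the pawn's coordinates to the clipboard, skipping zero locations to preserve the pre-elimination position.
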(event EventBeginPlay)

(event Collision|EventActorBeginOverlap
  (params (OtherActor :type Actor%20Object%20Reference)))

(event EventTick
  (params (DeltaSeconds :type Float%20%28single-precision%29))
  (bind _returnvalue (Game|GetPlayerController 0))
  (bind _returnvalue_2 (Pawn|GetControlledPawn _returnvalue))
  (bind (_returnvalue_x _returnvalue_y _returnvalue_z) (Math|Vector|BreakVector (Transformation|GetActorLocation _returnvalue_2)))
  (bind _returnvalue_location (Math|Transform|BreakTransform (Transformation|GetActorTransform _returnvalue_2)))
  ; Reset the bool
  (Variables|Default|SetIsOverlappingUV false)
  ; Make the cube stay with the player
  (Transformation|SetWorldLocation (Variables|Default|GetCube) _returnvalue_location false false)
  (bind _outactors (Actor|GetAllActorsOfClass "/Script/FortniteGame.FortUndergroundVolume"))
  (bind _array_element (Utilities|Array|ForEachLoopwithBreak _outactors)
    (:LoopBody
      ; If we are overlapping an UV, break out of the loop
      (if (Collision|IsOverlappingActor _returnvalue_2 _array_element)
        (Variables|Default|SetIsOverlappingUV true)))
    (:Completed
      _))
  ; If overlapping an UV, don't show the cube. Otherwise, show it.
  (if (Variables|Default|GetIsOverlappingUV)
    (Rendering|SetVisibility (Variables|Default|GetCube) false false)
    (else
      (Rendering|SetVisibility (Variables|Default|GetCube) true false)
      (if (not (Variables|Default|GetHideDebugText))
        (Development|PrintString "UNDERGROUND VOLUME NOT DETECTED!" true false "(R=1.000000,G=1.000000,B=0.000000,A=1.000000)" 1.0)
        ;[1 1 1] Build Bugitgo from the player location (pre-elimination) and copy it to the clipboard
        ;; If the player isn't eliminated
        (if (not (Utilities|Operators|Equal(==) :A (Transformation|GetActorLocation _returnvalue_2)))
          (Utility|CopytoClipboard (Utilities|String|Append "BugItGo " (Utilities|String|ToString(Text) (Utilities|Text|ToText(Float) _returnvalue_x "HalfToEven" false false 1 324 0 0)) " " (Utilities|String|ToString(Text) (Utilities|Text|ToText(Float) _returnvalue_y "HalfToEven" false false 1 324 0 0)) " " (Utilities|String|ToString(Text) (Utilities|Text|ToText(Float) _returnvalue_z "HalfToEven" false false 1 324 0 0)) " " (Utilities|String|ToString(Text) (Utilities|Text|ToText(Float) 0.0 "HalfToEven" false false)) " " (Utilities|String|ToString(Text) (Utilities|Text|ToText(Float) 0.0 "HalfToEven" false false)) " " (Utilities|String|ToString(Text) (Utilities|Text|ToText(Float) 0.0 "HalfToEven" false false)))))))))

(fn ConstructionScript)
